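# Post-process AnimBP for the Conscience female head that procedurally animates eye rotation and blinking on repurposed dyn_hair bones.
# The update event shapes game-time sine waves with arctangent into two snappy eye rotators and four brief blink alphas at differing rates.
# The AnimGraph chains additive bone-space modifiers for eye rotation and alpha-blended blink offsets.
# A final weighted parent constraint makes dyn_hair_03 follow jaw, lip and other face bones.
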AnimBlueprint /BRCosmetics/Characters/Player/Female/Medium/Heads/F_MED_Conscience_Head/Meshes/F_MED_Conscience_Post_AnimBP
target_skeleton=/Game/Characters/Player/Common/Fortnite_Base_Head/Fortnite_Base_Head_Export_Skeleton
graphs [
 AnimGraph {nodes [
   n1 = Get blink_03_alpha() -> (blink_03_alpha:double)
   n2 = Get blink_01_alpha() -> (blink_01_alpha:double)
   n3 = Get rot_eyes_01() -> (rot_eyes_01:Rotator)
   n4 = "Transform (Modify) Bone - Bone: dyn_hair_01_a"(ComponentPose=Local To Component(LocalPose=Input Pose()), Alpha=1, Translation="0,0,0", Rotation=@n3, Scale="1,1,1") {rotationMode=BMM_Additive, rotationSpace=BCS_BoneSpace, alphaBoolBlend {blendOption=Linear}, boneToModify {boneName=dyn_hair_01_a}}
   n5 = "Transform (Modify) Bone - Bone: dyn_hair_01_d"(ComponentPose=@n4, Alpha=1, Translation="0,0,0", Rotation=@n3, Scale="1,1,1") {rotationMode=BMM_Additive, rotationSpace=BCS_BoneSpace, alphaBoolBlend {blendOption=Linear}, boneToModify {boneName=dyn_hair_01_d}}
   n6 = Get rot_eyes_02() -> (rot_eyes_02:Rotator)
   n7 = "Transform (Modify) Bone - Bone: dyn_hair_01_g"(ComponentPose=@n5, Alpha=1, Translation="0,0,0", Rotation=@n6, Scale="1,1,1") {rotationMode=BMM_Additive, rotationSpace=BCS_BoneSpace, alphaBoolBlend {blendOption=Linear}, boneToModify {boneName=dyn_hair_01_g}}
   n8 = "Transform (Modify) Bone - Bone: dyn_hair_01_h"(ComponentPose=@n7, Alpha=1, Translation="0,0,0", Rotation=@n6, Scale="1,1,1") {rotationMode=BMM_Additive, rotationSpace=BCS_BoneSpace, alphaBoolBlend {blendOption=Linear}, boneToModify {boneName=dyn_hair_01_h}}
   n9 = "Transform (Modify) Bone - Bone: dyn_hair_01_b"(ComponentPose=@n8, Alpha=@n2, Translation="0.000,-0.479,0.100", Rotation="0,0,0", Scale="1,1,1") {
    translationMode=BMM_Additive
    translationSpace=BCS_BoneSpace
    alphaBoolBlend {blendOption=Linear}
    boneToModify {boneName=dyn_hair_01_b}
    translation {y=-0.479, z=0.1}
   }
   n10 = "Transform (Modify) Bone - Bone: dyn_hair_01_c"(ComponentPose=@n9, Alpha=@n2, Translation="0,-0.6,0.1", Rotation="0,0,0", Scale="1,1,1") {
    translationMode=BMM_Additive
    translationSpace=BCS_BoneSpace
    alphaBoolBlend {blendOption=Linear}
    boneToModify {boneName=dyn_hair_01_c}
    translation {y=-0.6, z=0.1}
   }
   n11 = Get blink_02_alpha() -> (blink_02_alpha:double)
   n12 = "Transform (Modify) Bone - Bone: dyn_hair_01_e"(ComponentPose=@n10, Alpha=@n11, Translation="0,-0.55,0.1", Rotation="0,0,0", Scale="1,1,1") {
    translationMode=BMM_Additive
    translationSpace=BCS_BoneSpace
    alphaBoolBlend {blendOption=Linear}
    boneToModify {boneName=dyn_hair_01_e}
    translation {y=-0.55, z=0.1}
   }
   n13 = "Transform (Modify) Bone - Bone: dyn_hair_01_f"(ComponentPose=@n12, Alpha=@n11, Translation="0,-0.6,0.1", Rotation="0,0,0", Scale="1,1,1") {
    translationMode=BMM_Additive
    translationSpace=BCS_BoneSpace
    alphaBoolBlend {blendOption=Linear}
    boneToModify {boneName=dyn_hair_01_f}
    translation {y=-0.6, z=0.1}
   }
   n14 = "Transform (Modify) Bone - Bone: dyn_hair_02_a"(ComponentPose=@n13, Alpha=@n1, Translation="0.000,-0.6,0.100", Rotation="0,0,0", Scale="1,1,1") {
    translationMode=BMM_Additive
    translationSpace=BCS_BoneSpace
    alphaBoolBlend {blendOption=Linear}
    boneToModify {boneName=dyn_hair_02_a}
    translation {y=-0.6, z=0.1}
   }
   n15 = "Transform (Modify) Bone - Bone: dyn_hair_02_b"(ComponentPose=@n14, Alpha=@n1, Translation="0,-0.6,0.1", Rotation="0,0,0", Scale="1,1,1") {
    translationMode=BMM_Additive
    translationSpace=BCS_BoneSpace
    alphaBoolBlend {blendOption=Linear}
    boneToModify {boneName=dyn_hair_02_b}
    translation {y=-0.6, z=0.1}
   }
   n16 = Get blink_04_alpha() -> (blink_04_alpha:double)
   n17 = "Transform (Modify) Bone - Bone: dyn_hair_02_c"(ComponentPose=@n15, Alpha=@n16, Translation="0,-0.6,0", Rotation="0,0,0", Scale="1,1,1") {
    rotationSpace=BCS_BoneSpace
    translationMode=BMM_Additive
    translationSpace=BCS_BoneSpace
    alphaBoolBlend {blendOption=Linear}
    boneToModify {boneName=dyn_hair_02_c}
    translation {y=-0.6}
   }
   n18 = "Transform (Modify) Bone - Bone: dyn_hair_02_d"(ComponentPose=@n17, Alpha=@n16, Translation="0,-0.6,0", Rotation="0,0,0", Scale="1,1,1") {
    translationMode=BMM_Additive
    translationSpace=BCS_BoneSpace
    alphaBoolBlend {blendOption=Linear}
    boneToModify {boneName=dyn_hair_02_d}
    translation {y=-0.6}
   }
   n19 = "Constraint - dyn_hair_03 (PPPPPPPPPPPPP)"!(
    ComponentPose=@n18
    Alpha=1
    ConstraintWeights_0=0.4747
    ConstraintWeights_1:float=0.28
    ConstraintWeights_2:float=0.0718
    ConstraintWeights_3:float=0.068
    ConstraintWeights_4:float=0.0417
    ConstraintWeights_5:float=0.0412
    ConstraintWeights_6:float=0.0161
    ConstraintWeights_7:float=0.0027
    ConstraintWeights_8:float=0.0018
    ConstraintWeights_9:float=0.0017
    ConstraintWeights_10:float=0.0001
    ConstraintWeights_11:float=0.0001
    ConstraintWeights_12:float=0
   ) -> (Pose) {
    alphaBoolBlend {blendOption=Linear}
    boneToModify {boneName=dyn_hair_03}
    constraintSetup [
     {targetBone {boneName=C_jaw}}
     {targetBone {boneName=C_lip_lower_mid}}
     {targetBone {boneName=L_lip_lower_outer}}
     {targetBone {boneName=R_lip_lower_outer}}
     {targetBone {boneName=L_lip_corner}}
     {targetBone {boneName=R_lip_corner}}
     {targetBone {boneName=head}}
     {targetBone {boneName=R_lip_upper_outer}}
     {targetBone {boneName=L_lip_upper_outer}}
     {targetBone {boneName=C_lip_upper_mid}}
     {targetBone {boneName=R_cheek_inner}}
     {targetBone {boneName=L_cheek_inner}}
     {targetBone {boneName=C_nose_bridge}}
    ]
    constraintWeights [1,1,1,1,1,1,1,1,1,1,1,1,1]
   }
   Output Pose(Result=Component To Local(ComponentPose=@n19)) {layerGroup=None}
   Reroute Node(InputPin:Wildcard) -> (OutputPin:Wildcard)
  ]}
 EventGraph {nodes [
   Event BlueprintUpdateAnimation()
   > n1 = Sequence!(execute) -> (then_0, then_1, then_2:Exec, then_3:Exec, then_4:Exec, then_5:Exec)
   n2 = GetGameTimeInSeconds()
   n3 = float * float(A=@n2, B=0.75)
   n4 = "Atan (Radians)"!(A:double=float * float(A="Sin (Radians)"(A=@n3), B=50)) -> (ReturnValue:double)
   n5 = float * float(A=@n2, B=0.5)
   n6 = "Atan (Radians)"(A=float * float(A="Sin (Radians)"(A=@n5), B=25))
   > Set rot_eyes_01(rot_eyes_01:Rotator=MakeRotator(Roll=float * float(A=@n6, B=1.5), Yaw=float * float(A=@n4, B=0.75))) -> (Output_Get:Rotator)
   n7 = "Sin (Radians)"(A=float * float(A=@n2, B=0.55))
   n8 = "Atan (Radians)"(A=float * float(A="Clamp (Float)"(Value=@n7, Min=-0.85, Max=0.85), B=90))
   n9 = float * float(A=@n8, B=@n8)
   Set blink_02_alpha(execute=@n1.then_3, blink_02_alpha:double=float + float(A=float * float(A=@n9, B=-0.4), B=1)) -> (Output_Get:double)
   n10 = "Sin (Radians)"(A=float * float(A=@n2, B=0.7))
   n11 = "Atan (Radians)"(A=float * float(A="Clamp (Float)"(Value=@n10, Min=-0.85, Max=0.85), B=95))
   n12 = float * float(A=@n11, B=@n11)
   Set blink_03_alpha(execute=@n1.then_4, blink_03_alpha:double=float + float(A=float * float(A=@n12, B=-0.4), B=1)) -> (Output_Get:double)
   n13 = "Sin (Radians)"(A=float * float(A=@n2, B=0.65))
   n14 = "Atan (Radians)"(A=float * float(A="Clamp (Float)"(Value=@n13, Min=-0.85, Max=0.85), B=105))
   n15 = float * float(A=@n14, B=@n14)
   Set blink_04_alpha(execute=@n1.then_5, blink_04_alpha:double=float + float(A=float * float(A=@n15, B=-0.4), B=1)) -> (Output_Get:double)
   n16 = float * float(A=@n2, B=0.6)
   n17 = "Atan (Radians)"(A=float * float(A="Sin (Radians)"(A=@n16), B=35))
   n18 = float * float(A=@n2, B=0.75)
   n19 = "Atan (Radians)"(A=float * float(A="Sin (Radians)"(A=@n18), B=50))
   Set rot_eyes_02(execute=@n1.then_1, rot_eyes_02:Rotator=MakeRotator(Roll=float * float(A=@n17, B=1.5), Yaw=float * float(A=@n19, B=0.75))) -> (Output_Get:Rotator)
   n20 = "Sin (Radians)"(A=float * float(A=@n2, B=0.75))
   n21 = "Atan (Radians)"(A=float * float(A="Clamp (Float)"(Value=@n20, Min=-0.85, Max=0.85), B=100))
   n22 = float * float(A=@n21, B=@n21)
   Set blink_01_alpha(execute=@n1.then_2, blink_01_alpha:double=float + float(A=float * float(A=@n22, B=-0.4), B=1)) -> (Output_Get:double)
  ]}
]
variables [
 blink_01_alpha {type=number}
 blink_02_alpha {type=number}
 blink_03_alpha {type=number}
 blink_04_alpha {type=number}
 rot_eyes_01 {type=Rotator}
 rot_eyes_02 {type=Rotator}
]
